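; Pilgrim quickplay view model that tracks whether the More Options container is open, driven by quest screen visibility.
; When the global event router becomes available, it waits for the Pilgrim event router on the owning player controller.
; It then listens for the Sparks quest screen visibility event for the player state and copies its visibility flag into bIsMoreOptionsContainerOpened.
; The listener handle is registered with the view model's event tracking.
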(event Sparks|ViewModel|EventBP_OnGlobalEventRouterAvailable
  (params (EventRouter :type Gameplay%20Event%20Router%20Component%20Object%20Reference) (PlayerState :type Fort%20Player%20State%20Object%20Reference) (OwningLocalPlayer :type Local%20Player%20Object%20Reference) (OwningPlayerController :type Fort%20Player%20Controller%20Object%20Reference))
  (bind _eventrouter (Pilgrim|PilgrimWaitForEventRouter OwningPlayerController)
    (:OnEventRouterAvailable
      (bind (_asynctaskproxy _payload) (GameplayEventRouter|StartListeningToEvent OwningPlayerController _eventrouter "/Script/SparksCoreUI.SparksEvent_QuestScreenVisibilityChanged" PlayerState)
        (:then
          (Sparks|ViewModel|TrackEventHandle _asynctaskproxy))
        (:OnEventReceived
          (bind _payload_bisvisible (Utilities|Struct|BreakSparksEventQuestScreenVisibilityChanged _payload))
          (Variables|Default|SetbIsMoreOptionsContainerOpened _payload_bisvisible))))))
