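; View model that builds the solo player's stat list for Pilgrim Quickplay song results.
; When the event router becomes available, it caches performance data from the player state component and branches on instrument choice.
; Pro Vocals gets phrase counts per config rating tier.
; Other instruments get notes, accuracy tiers, misses, strikes, streak and early/late timing as stat entry objects, with the total approximated as hit plus passed notes.
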(event Sparks|ViewModel|EventBP_OnGlobalEventRouterAvailable
  (params (EventRouter :type Gameplay%20Event%20Router%20Component%20Object%20Reference) (PlayerState :type Fort%20Player%20State%20Object%20Reference) (OwningLocalPlayer :type Local%20Player%20Object%20Reference) (OwningPlayerController :type Fort%20Player%20Controller%20Object%20Reference))
  (bind _returnvalue (Class|PilgrimQuickplayPlayerStateComponent|GetPilgrimQuickplayPlayerStateComponent PlayerState))
  (Utilities|IsValid _returnvalue
    (:Is%20Valid
      (Variables|Default|SetCurrentPerformanceData (Class|PilgrimQuickplayPlayerStateComponent|GetPerformanceData _returnvalue))
      (if (Utilities|Enum|Equal(Enum) (Class|PilgrimCorePlayerStateComponent|GetInstrumentChoice _returnvalue) "TrackProVocals")
        (Variables|Default|SetPlayerStats (ProVocalsCase|GetSongStats OwningPlayerController))
        (else
          (bind _playerstatsarray (GeneralCase|UpdateAllPlayerStatItems))
          (Variables|Default|SetPlayerStats _playerstatsarray))))))

(fn AddAverageOffsetItem
  (bind (_currentperformancedata_accuracy _currentperformancedata_averageoffset) (Utilities|Struct|BreakPilgrimQuickplayPlayerPerformanceData (Variables|GetCurrentPerformanceData_Accuracy)))
  (bind _returnvalue (Math|Float|Round _currentperformancedata_averageoffset))
  (bind _returnvalue_2 (> _returnvalue 0))
  ;[0.5117 1 0.9099] Show the average offset (calibration) to help players no if their hits are early or late. Will show a + or - and Early or Late depending on if the offset is positive or negative.
  (GeneralCase|QueueStatItem "NSLOCTEXT(\"[1F7929D7FA6FC94B5077D0FA8795E7CA]\", \"5BC9AD4642FA428F4F1FBDB50E6DDC3B\", \"Avg. input time\")" (Utilities|Text|FormatText "NSLOCTEXT(\"[1F7929D7FA6FC94B5077D0FA8795E7CA]\", \"32D8E05648A4A70D14D27CB82AC89004\", \"{sign}{offset} ms\")" (select _returnvalue_2 "NSLOCTEXT(\"[1F7929D7FA6FC94B5077D0FA8795E7CA]\", \"3B3D5082469CA0FCDD4E83A843730CC2\", \"+\")" "") _returnvalue))
  ;[0.6246 1 0.9834] If the average offset is not equal to 0, add an additional stat line with no label that indicates if the offset was early or late
  (if (!= _returnvalue 0)
    (GeneralCase|QueueStatItem "" (select _returnvalue_2 "NSLOCTEXT(\"[1F7929D7FA6FC94B5077D0FA8795E7CA]\", \"3739A6B9405BC07951C1D0AECD7E0328\", \"(Late)\")" "NSLOCTEXT(\"[1F7929D7FA6FC94B5077D0FA8795E7CA]\", \"4AD2406C4F3CF0D881258DA19804C368\", \"(Early)\")"))))

(fn AddLongestStreakStatItem
  (bind (_currentperformancedata_accuracy _currentperformancedata_averageoffset _currentperformancedata_standarddeviation _currentperformancedata_fullcombo _currentperformancedata_noteshit _currentperformancedata_notespassed _currentperformancedata_notesmissed _currentperformancedata_totalnotes _currentperformancedata_longeststreak) (Utilities|Struct|BreakPilgrimQuickplayPlayerPerformanceData (Variables|GetCurrentPerformanceData_Accuracy)))
  (GeneralCase|QueueStatItem "NSLOCTEXT(\"[1F7929D7FA6FC94B5077D0FA8795E7CA]\", \"2A1BD6724ABD51B95E0C5CAED94793A8\", \"Longest streak\")" (Utilities|Text|ToText(Integer) _currentperformancedata_longeststreak)))

(fn AddNotesHitAccuracyTiersStatItems
  (bind (_currentperformancedata_accuracy _currentperformancedata_averageoffset _currentperformancedata_standarddeviation _currentperformancedata_fullcombo _currentperformancedata_noteshit _currentperformancedata_notespassed _currentperformancedata_notesmissed _currentperformancedata_totalnotes _currentperformancedata_longeststreak _currentperformancedata_timeinoverdrivems _currentperformancedata_histogramsamples _currentperformancedata_accuracytiercounts) (Utilities|Struct|BreakPilgrimQuickplayPlayerPerformanceData (Variables|GetCurrentPerformanceData_Accuracy)))
  ;[0.625 1 1] Adds the number of notes hit for each accuracy tier (e.g. perfect, good)
  (for (_array_element _array_index) _currentperformancedata_accuracytiercounts
    (GeneralCase|QueueStatItem (Utilities|Select :Index _array_index :Option%200 (Variables|Default|GetPerfectNotesText) :Option%201 (Variables|Default|GetGoodNotesText)) (Utilities|Text|FormatText "NSLOCTEXT(\"[1F7929D7FA6FC94B5077D0FA8795E7CA]\", \"97762EFF42AC71B91257938F60F4AD92\", \"{hit} ({percent})\")" _array_element (Utilities|Text|AsPercent (/ (Math|Conversions|ToFloat(Integer) _array_element) (GeneralCase|GetTotalNotesinSong)) "ToZero" false true 1 324 0 0)))))

(fn AddPassedNotesHitStatItem
  (bind (_currentperformancedata_accuracy _currentperformancedata_averageoffset _currentperformancedata_standarddeviation _currentperformancedata_fullcombo _currentperformancedata_noteshit _currentperformancedata_notespassed) (Utilities|Struct|BreakPilgrimQuickplayPlayerPerformanceData (Variables|GetCurrentPerformanceData_Accuracy)))
  ;[0.625 1 1] Note: even we display passed notes as 'missed notes' to the player
  (GeneralCase|QueueStatItem "NSLOCTEXT(\"[1F7929D7FA6FC94B5077D0FA8795E7CA]\", \"466FAE14418B194E4E595C9E693206A8\", \"Missed notes\")" (Utilities|Text|FormatText "NSLOCTEXT(\"[1F7929D7FA6FC94B5077D0FA8795E7CA]\", \"589060884FCB4B78AB135B8B91366BBD\", \"{stat} ({percent})\")" _currentperformancedata_notespassed (Utilities|Text|AsPercent (/ (Math|Conversions|ToFloat(Integer) _currentperformancedata_notespassed) (GeneralCase|GetTotalNotesinSong)) "FromZero" false true 1 324 0 0))))

(fn AddStrikesStatItem
  (bind (_currentperformancedata_accuracy _currentperformancedata_averageoffset _currentperformancedata_standarddeviation _currentperformancedata_fullcombo _currentperformancedata_noteshit _currentperformancedata_notespassed _currentperformancedata_notesmissed) (Utilities|Struct|BreakPilgrimQuickplayPlayerPerformanceData (Variables|GetCurrentPerformanceData_Accuracy)))
  ;[0.625 1 1] Show missed notes (Strikes)
  (GeneralCase|QueueStatItem "NSLOCTEXT(\"[1F7929D7FA6FC94B5077D0FA8795E7CA]\", \"9A579D304608F192033D5FA3C34DE939\", \"Strikes\")" (Utilities|Text|ToText(Integer) _currentperformancedata_notesmissed)))

(fn AddTotalNotesItem
  (GeneralCase|QueueStatItem "NSLOCTEXT(\"[1F7929D7FA6FC94B5077D0FA8795E7CA]\", \"3358F404435E66BCC62B07B3DA3ABC23\", \"Total notes\")" (Utilities|Text|ToText(Integer) (GeneralCase|GetTotalNotesinSong))))

(fn AppendStatEntry
  (params (StatTitle :type Text) (StatLabel :type Text) (ContentType :type E_BandStats_StatEntry_ContentType%20Enum :default "NewEnumerator0") (AppendTo :type Array%20of%20BP%20Band%20Stats%20Stat%20Entry%20Data%20Object%20References))
  (bind _returnvalue (Game|ConstructObjectfromClass "/PilgrimQuickplay/Widgets/Results/ListEntryBlueprints/BP_BandStats_Stat_EntryData.BP_BandStats_Stat_EntryData_C"))
  (Variables|/PilgrimQuickplay/Widgets/Results/ListEntryBlueprints/BP_BandStats_Stat_EntryData.BP_BandStats_Stat_EntryData_C|SetEntryData_StatTitle_9_6F2BA55942B3781BA284D29ACE297F66 (Utilities|Struct|MakeSBandStatsStatEntry StatTitle StatLabel ContentType) _returnvalue)
  (Utilities|Array|Add AppendTo _returnvalue))

(fn GetSongStats
  (params (PlayerController :type Fort%20Player%20Controller%20Object%20Reference))
  (bind (_currentperformancedata_accuracy _currentperformancedata_averageoffset _currentperformancedata_standarddeviation _currentperformancedata_fullcombo _currentperformancedata_noteshit _currentperformancedata_notespassed _currentperformancedata_notesmissed _currentperformancedata_totalnotes _currentperformancedata_longeststreak _currentperformancedata_timeinoverdrivems _currentperformancedata_histogramsamples _currentperformancedata_accuracytiercounts _currentperformancedata_sectionscompletedbytier) (Utilities|Struct|BreakPilgrimQuickplayPlayerPerformanceData (Variables|GetCurrentPerformanceData_Accuracy)))
  ;[0 0 0 0.5] Config data, for indexing with
  (bind _returnvalue (Class|PlayspaceLibrary|GetPlayspaceforActor PlayerController PlayerController))
  (bind _aspilgrim_playspace (Utilities|Casting|CastToPilgrimPlayspace _returnvalue))
  (bind _returnvalue_2 (Class|PilgrimPlayspace|GetConfig _aspilgrim_playspace))
  ;[0 0 0 0.5] For every Config entry, look up the Results
  (for (_array_element _array_index) (.ProVocalsSectionRatings (Utilities|Struct|BreakPilgrimGameConfigSettings (Variables|/Script/PilgrimCoreRuntime.PilgrimGameConfig|GetConfigSettings _returnvalue_2)))
    (bind _array_element_name (Utilities|Struct|BreakProVocalsSectionRating _array_element))
    (ProVocalsCase|AppendStatEntry (Utilities|Text|FormatText "NSLOCTEXT(\"[1F7929D7FA6FC94B5077D0FA8795E7CA]\", \"B93A57704A8808C30A612C83CC6A95C2\", \"{tier} phrases\")" _array_element_name) (Utilities|Text|FormatText "NSLOCTEXT(\"[1F7929D7FA6FC94B5077D0FA8795E7CA]\", \"E7B119D44719A219E17141BE55F03ABB\", \"{hit} ({percent})\")" (Utilities|Text|ToText(Integer) (Utilities|Array|Get(acopy) _currentperformancedata_sectionscompletedbytier _array_index)) (Utilities|Text|AsPercent (/ (Math|Conversions|ToFloat(Integer) (Utilities|Array|Get(acopy) _currentperformancedata_sectionscompletedbytier _array_index)) _currentperformancedata_totalnotes) "ToZero" false true 1 324 0 0)) "NewEnumerator0" (LocalVariables|GetStats)))
  ;[0 0 0 0.5] Done
  (return (LocalVariables|GetStats)))

(fn GetTotalNotesInSong
  (bind (_currentperformancedata_accuracy _currentperformancedata_averageoffset _currentperformancedata_standarddeviation _currentperformancedata_fullcombo _currentperformancedata_noteshit _currentperformancedata_notespassed) (Utilities|Struct|BreakPilgrimQuickplayPlayerPerformanceData (Variables|GetCurrentPerformanceData_Accuracy)))
  ;[0.625 1 1] Replace with something more accurate: attempt to get a count of the total notes in the song by adding all notes hit and passed
  (return (+ _currentperformancedata_noteshit _currentperformancedata_notespassed)))

(fn QueueStatItem
  (params (StatTitle :type Text) (StatValue :type Text) (ContentType :type E_BandStats_StatEntry_ContentType%20Enum :default "NewEnumerator0"))
  (bind _returnvalue (Game|ConstructObjectfromClass "/PilgrimQuickplay/Widgets/Results/ListEntryBlueprints/BP_BandStats_Stat_EntryData.BP_BandStats_Stat_EntryData_C"))
  (Variables|/PilgrimQuickplay/Widgets/Results/ListEntryBlueprints/BP_BandStats_Stat_EntryData.BP_BandStats_Stat_EntryData_C|SetEntryData_StatTitle_9_6F2BA55942B3781BA284D29ACE297F66 (Utilities|Struct|MakeSBandStatsStatEntry StatTitle StatValue ContentType) _returnvalue)
  (Utilities|Array|Add (Variables|Default|GetInternal%20StatListData) _returnvalue))

(fn UpdateAllPlayerStatItems
  (Utilities|Array|Clear (Variables|Default|GetInternal%20StatListData))
  (Utilities|FlowControl|Sequence
    (:then_0
      (GeneralCase|AddTotalNotesItem))
    (:then_1
      (GeneralCase|AddNotesHitAccuracyTiersStatItems))
    (:then_2
      (GeneralCase|AddPassedNotesHitStatItem))
    (:then_3
      (GeneralCase|AddStrikesStatItem))
    (:then_4
      (GeneralCase|AddLongestStreakStatItem))
    (:then_5
      (GeneralCase|AddAverageOffsetItem))
    (:then_6
      (return (Variables|Default|GetInternal%20StatListData)))))
