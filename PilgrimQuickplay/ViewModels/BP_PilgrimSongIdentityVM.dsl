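; View model that exposes the current song's artist name and uppercased title for UI in Pilgrim Quickplay.
; When the global event router becomes available, it sets both values from the song in the player's band state.
; It then waits for the Pilgrim event router and listens for the MBI SongLoaded event, refreshing them from each loaded song's metadata.
; The listener handle is tracked by the view model.
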(event Sparks|ViewModel|EventBP_OnGlobalEventRouterAvailable
  (params (EventRouter :type Gameplay%20Event%20Router%20Component%20Object%20Reference) (PlayerState :type Fort%20Player%20State%20Object%20Reference) (OwningLocalPlayer :type Local%20Player%20Object%20Reference) (OwningPlayerController :type Fort%20Player%20Controller%20Object%20Reference))
  (Utilities|FlowControl|Sequence
    (:then_0
      (CallFunction|SetSongIdentity (Class|PilgrimCoreBandStateComponent|GetCurrentSong (Class|PilgrimQuickplayBandStateComponent|GetBandState OwningPlayerController))))
    (:then_1
      (bind _eventrouter (Pilgrim|PilgrimWaitForEventRouter OwningPlayerController)
        (:OnEventRouterAvailable
          (bind (_asynctaskproxy _payload) (GameplayEventRouter|StartListeningToEvent _eventrouter _eventrouter "/Script/PilgrimCoreRuntime.PilgrimEvent_MBI_SongLoaded")
            (:then
              (Sparks|ViewModel|TrackEventHandle _asynctaskproxy))
            (:OnEventReceived
              (bind _payload_loadedsong (Utilities|Struct|BreakPilgrimEventMBISongLoaded _payload))
              (CallFunction|SetSongIdentity _payload_loadedsong))))))))

(fn SetSongIdentity
  (params (SongMetadata :type Pilgrim%20Song%20Metadata%20Object%20Reference))
  (Variables|Default|SetArtist%20Name (Meta|GetArtistText SongMetadata))
  (Variables|Default|SetSong%20Title (Utilities|Text|TexttoUpper (Meta|GetTitleText SongMetadata))))
